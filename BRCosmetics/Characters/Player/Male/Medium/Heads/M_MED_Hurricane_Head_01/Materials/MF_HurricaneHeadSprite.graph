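# Material function for the Hurricane character head that draws an animated, glowing fiery core sprite and adds it to the input material's emissive.
# Screen-aligned UVs centered on HeadSocketLocation scale with FOV and object radius, faking a sprite without extra geometry.
# Sphere masks shape a noise-distorted core, panning flames, hash-flickered outer glow and two scrolling ring layers, one polar-distorted.
# Time-of-day color, CloudsMask and inverted DirtMask attenuate the sum; a virtual-plane UV branch sits unconnected.
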MaterialFunction /BRCosmetics/Characters/Player/Male/Medium/Heads/M_MED_Hurricane_Head_01/Materials/MF_HurricaneHeadSprite
graph {nodes [
  Constant() {r=1}
  n1 = GetMaterialAttributes(None=FunctionInput() {inputName=Material, inputType=FunctionInput_MaterialAttributes}) {attributeGetTypes [#1]}
  n2 = VectorParameter() {parameterName=HC_Core_Color, defaultValue=<rgba 1 0.4 0.05 0>}
  n3 = VectorParameter() {parameterName=HeadSocketLocation, defaultValue=<rgba 0 0 0 1>}
  n4 = MaterialFunctionCall!(Center Point World Position=@n3.RGB) {materialFunction=/Engine/Functions/Engine_MaterialFunctions02/UVs/WorldSpaceAlignedScreenCoordinates}
  n5 = ScalarParameter() {defaultValue=12, parameterName=UV_Scale}
  n6 = MaterialFunctionCall() {materialFunction=/Engine/Functions/Engine_MaterialFunctions02/Utility/FOV}
  n7 = Constant() {r=0.414}
  n8 = Constant() {r=1.091}
  n9 = Constant()
  n10 = Constant() {r=1}
  n11 = LinearInterpolate(
   A=Multiply(A=@n5) {constB=2}
   B=@n5
   Alpha=Saturate(None=MaterialFunctionCall(Input=@n6."FOV ", Input Low=@n7, Input High=@n8, Target Low=@n9, Target High=@n10) {materialFunction=&RemapValueRange}.Result)
  )
  n12 = ObjectRadius()
  n13 = ScalarParameter() {defaultValue=96.9905, group=Neon, parameterName=ActorRadius, sortPriority=0}
  n14 = LinearInterpolate(A=@n11, B=Multiply(A=@n11, B=Divide(A=@n12, B=@n13)), Alpha=ScalarParameter() {defaultValue=1, parameterName=UseObjectScale})
  n15 = MaterialFunctionCall(UVs=@n4.Offset UVs, Texture Scale=@n14) {materialFunction=&ScaleUVsByCenter}
  n16 = Subtract(A=Multiply(A=@n15.UVs) {constB=2})
  n17 = Constant2Vector()
  n18 = ComponentMask(None=VectorParameter() {parameterName=HC_CoreDistortionNoise_Tiling, defaultValue=<rgba 0.5 0.5 0.5 1>}.RGB) {g=true, r=true}
  n19 = VectorParameter() {parameterName=HC_CoreDistortionNoise_PanSpeedAndDirection, defaultValue=<rgba 0.15 0.15 1 1>}
  n20 = TextureSampleParameter2D(UVs=Panner(Coordinate=Multiply(A=@n15.UVs, B=@n18), Speed=ComponentMask(None=@n19.RGB) {g=true, r=true})) {parameterName=DistortTexture, samplerType=SAMPLERTYPE_LinearColor, texture=&T_NoisePack_LinearColor}
  n21 = LinearInterpolate(Alpha=Multiply(A=@n20.R, B=ScalarParameter() {defaultValue=1, parameterName=HC_CoreDistortionStrength})) {constA=0.5}
  n22 = Constant() {r=0.1}
  n23 = Clamp(None=Power(Base=SphereMask(A=@n16, B=@n17, Radius=@n21, Hardness=@n22)) {constExponent=4})
  n24 = Power(Base=Multiply(A=Multiply(A=@n2.RGB, B=@n23), B=ScalarParameter() {defaultValue=25, parameterName=HC_CoreGlowStrength}))
  n25 = Subtract(A=Multiply(A=@n15.UVs) {constB=2})
  n26 = ComponentMask(None=VectorParameter() {parameterName=HC_FlameNoise_Tiling, defaultValue=<rgba 1 1 1 1>}.RGB) {g=true, r=true}
  n27 = ComponentMask(None=Constant3Vector() {constant=<rgba 1 1 1 0>}) {g=true, r=true}
  n28 = VectorParameter() {parameterName=HC_CoreFlame_PanSpeedAndDirection, defaultValue=<rgba -0.1 0.25 1 1>}
  n29 = TextureSampleParameter2D(UVs=Panner(Coordinate=Multiply(A=@n15.UVs, B=@n27), Speed=ComponentMask(None=@n28.RGB) {g=true, r=true})) {parameterName=DistortTexture, samplerType=SAMPLERTYPE_LinearColor, texture=&T_NoisePack_LinearColor}
  n30 = Subtract(A=Multiply(A=@n29.RGB) {constB=2})
  n31 = ComponentMask(None=@n30) {g=true, r=true}
  n32 = ScalarParameter() {defaultValue=0.15, parameterName=HC_FlameDistortionStrength}
  n33 = Panner(Coordinate=Add(A=Multiply(A=@n25, B=@n26), B=Multiply(A=@n31, B=@n32))) {speedX=0.15, speedY=-0.2}
  n34 = Clamp(None=Multiply(
    A=TextureSampleParameter2D(UVs=@n33) {parameterName=FlameNoise, texture=/Game/Global/Textures/Default/Tileable_Patterns/T_WavePattern_01}.R
    B=ScalarParameter() {defaultValue=3, parameterName=HC_FlameOpacity}
   ) {constB=2})
  n35 = LinearInterpolate(
   A=VectorParameter() {parameterName=HC_Flame_Color_02, defaultValue=<rgba 1 0.214 0 1>}.RGB
   B=VectorParameter() {parameterName=HC_Flame_Color_01, defaultValue=<rgba 1 0.7267 0 1>}.RGB
   Alpha=@n34
  )
  n36 = Subtract(A=Multiply(A=@n15.UVs) {constB=2})
  n37 = Constant2Vector()
  n38 = ScalarParameter() {defaultValue=1, parameterName=HC_CoreFlameDistortion_Strength}
  n39 = LinearInterpolate(A=Constant() {r=1}, B=Constant() {r=0.5}, Alpha=ComponentMask(None=Multiply(A=@n30, B=@n38)) {b=true})
  n40 = Constant()
  n41 = Power(Base=Divide(A=SphereMask(A=@n36, B=@n37, Radius=@n39, Hardness=@n40)))
  n42 = Multiply(A=Multiply(A=Multiply(A=@n34, B=@n35), B=Clamp(None=@n41)), B=ScalarParameter() {defaultValue=25, parameterName=HC_Flame_Glowstrength})
  n43 = Multiply(A=@n15.UVs) {constB=2}
  n44 = ScalarParameter() {defaultValue=1, parameterName=GlowMinRadius}
  n45 = ScalarParameter() {defaultValue=2, parameterName=GlowMaxRadius}
  n46 = MaterialFunctionCall(Float=Multiply(A=Time()) {constB=10}) {materialFunction=&MF_Hash}
  n47 = Power(Base=SphereMask(A=Subtract(A=@n43), B=Constant2Vector(), Radius=LinearInterpolate(A=@n44, B=@n45, Alpha=@n46.Float), Hardness=Constant()))
  n48 = Multiply(
   A=Multiply(A=VectorParameter() {parameterName=HC_CoreGlow_Color, defaultValue=<rgba 1 0.0832 0 1>}.RGB, B=Clamp(None=@n47))
   B=ScalarParameter() {defaultValue=1, parameterName=HC_OuterGlowStrength}
  )
  n49 = Constant3Vector() {constant=<rgba 0.4612 0 1 0>}
  n50 = SphereMask(A=Subtract(A=Multiply(A=@n15.UVs) {constB=2}), B=Constant2Vector(), Radius=Constant() {r=1.75}, Hardness=Constant())
  n51 = Power(Base=@n50)
  n52 = ScalarParameter() {defaultValue=0.25, parameterName=HC_CoreWaveTimeDistortion}
  n53 = Multiply(A=Add(A=Time(), B=Multiply(A=@n20.G, B=@n52)), B=ScalarParameter() {defaultValue=0.35, parameterName=HC_CoreWaves_Speed})
  n54 = Frac(None=Multiply(A=Add(A=@n50, B=@n53), B=ScalarParameter() {defaultValue=8, parameterName=HC_CoreWave_Frequency}))
  n55 = Multiply(A=Power(Base=Subtract(A=@n54) {constB=0.5})) {constB=4}
  n56 = Multiply(A=@n15.UVs) {constB=2}
  n57 = Multiply(A=Clamp(None=Multiply(A=@n51, B=@n55)), B=SphereMask(A=Subtract(A=@n56), B=Constant2Vector(), Radius=Constant() {r=2}, Hardness=Constant()))
  n58 = Constant3Vector() {constant=<rgba 0 0.2303 1 0>}
  n59 = SphereMask(A=Subtract(A=Multiply(A=@n15.UVs) {constB=2}), B=Constant2Vector(), Radius=Constant() {r=1.75}, Hardness=Constant())
  n60 = Power(Base=@n59)
  n61 = Time()
  n62 = Subtract(A=Multiply(A=@n15.UVs) {constB=2})
  n63 = Arctangent2Fast(Y=ComponentMask(None=@n62) {g=true}, X=ComponentMask(None=@n62) {r=true})
  n64 = AppendVector(A=Divide(A=ComponentMask(None=@n63) {r=true}), B=Constant())
  n65 = ComponentMask(None=VectorParameter() {parameterName=HC_CoreFrequencyRingsDistortionNoise_Tiling, defaultValue=<rgba 4 1 4 1>}.RGB) {g=true, r=true}
  n66 = VectorParameter() {parameterName=HC_CoreFrequencyRings_PanSpeedAndDirection, defaultValue=<rgba 0.15 0.15 1 1>}
  n67 = TextureSampleParameter2D(UVs=Panner(Coordinate=Multiply(A=@n64, B=@n65), Speed=ComponentMask(None=@n66.RGB) {g=true, r=true})) {parameterName=DistortTexture, samplerType=SAMPLERTYPE_LinearColor, texture=&T_NoisePack_LinearColor}
  n68 = SphereMask(A=Subtract(A=Multiply(A=@n15.UVs) {constB=2}), B=Constant2Vector(), Radius=Constant() {r=2}, Hardness=Constant())
  n69 = Multiply(A=Time())
  n70 = Multiply(
   A=Multiply(A=ScalarParameter() {defaultValue=0.25, parameterName=HC_CoreRingsTimeDistortion}, B=@n68)
   B=ConstantBiasScale(None=MaterialFunctionCall(Float=@n69) {materialFunction=&MF_Hash}.Float)
  )
  n71 = Multiply(A=Add(A=Add(A=@n61) {constB=0.25}, B=Multiply(A=@n67.R, B=@n70)), B=ScalarParameter() {defaultValue=0.55, parameterName=HC_CoreRings_Speed})
  n72 = Frac(None=Multiply(A=Add(A=@n59, B=@n71), B=ScalarParameter() {defaultValue=7, parameterName=HC_CoreRings_Frequency}))
  n73 = Multiply(A=Power(Base=Subtract(A=@n72) {constB=0.5})) {constB=10}
  n74 = Multiply(A=Clamp(None=Multiply(A=@n60, B=@n73)), B=@n68)
  n75 = Add(
   A=Multiply(A=Multiply(A=@n49, B=@n57), B=ScalarParameter() {defaultValue=4, parameterName=HC_RingsGlowStrength}) {constB=0.5}
   B=Multiply(A=Multiply(A=@n58, B=@n74), B=ScalarParameter() {defaultValue=4, parameterName=HC_RingsGlowStrength}) {constB=0.5}
  )
  n76 = Add(A=@n24, B=Add(A=Add(A=@n24, B=@n42), B=Add(A=@n48, B=@n75)))
  n77 = MaterialFunctionCall(
   Color_Morning=VectorParameter() {parameterName=HC_TOD_EMissiveMultiplier_Morning, defaultValue=<rgba 1 1 1 1>}.RGB
   Color_Day=VectorParameter() {parameterName=HC_TOD_EMissiveMultiplier_Day, defaultValue=<rgba 1 1 1 1>}.RGB
   Color_Evening=VectorParameter() {parameterName=HC_TOD_EMissiveMultiplier_Evening, defaultValue=<rgba 1 1 1 1>}.RGB
   Color_Night=VectorParameter() {parameterName=HC_TOD_EMissiveMultiplier_Night, defaultValue=<rgba 0.25 0.25 0.25 1>}.RGB
  ) {materialFunction=&Color_Time_of_Day}
  n78 = FunctionInput(Preview=Constant() {r=1}) {bUsePreviewValueAsDefault=true, inputName=CloudsMask, inputType=FunctionInput_Scalar}
  n79 = Constant() {r=1}
  n80 = Multiply(
   A=Multiply(A=Multiply(A=@n76, B=@n77.Result), B=Clamp(None=@n78) {maxDefault=0.75, minDefault=0.25})
   B=OneMinus(None=FunctionInput(Preview=@n79) {bUsePreviewValueAsDefault=true, inputName=DirtMask, inputType=FunctionInput_Scalar})
  )
  FunctionOutput(None=SetMaterialAttributes!(MaterialAttributes=@n1.MaterialAttributes, EmissiveColor=Add(A=@n1.EmissiveColor, B=@n80)) {attributeSetTypes [#1]})
  n81 = CollectionParameter() {
   collection=&FortniteMaterialParameters
   parameterName=FortniteActiveTimeOfDay
  }
  MaterialFunctionCall(Float4=@n81) {materialFunction=&BreakOutFloat4Components}
  n82 = ComponentMask(None=VectorParameter() {group=ViperZero, parameterName=CorePosition, defaultValue=<rgba 6 6 0 1>}.RGB) {g=true, r=true}
  Multiply(A=Add(A=@n82, B=@n4.Offset UVs), B=@n14)
  n83 = Normalize(VectorInput=Subtract(A=ObjectPositionWS(), B=CameraPositionWS()))
  n84 = CrossProduct(A=@n83, B=VectorParameter() {parameterName=Axis, defaultValue=<rgba 0 0 1 1>}.RGB)
  n85 = MaterialFunctionCall(Float3=PreSkinnedLocalBounds().Half Extents) {materialFunction=&BreakOutFloat3Components}
  n86 = Multiply(
   A=MaterialFunctionCall() {materialFunction=/Engine/Functions/Engine_MaterialFunctions02/WorldPositionOffset/ObjectScale}.Scale X
   B=Max(A=Max(A=@n85.R, B=@n85.R), B=@n85.B)
  )
  n87 = ScalarParameter() {defaultValue=1.5, parameterName=UVRadialScaleZero}
  Subtract(
   A=MaterialFunctionCall!(Plane Normal Axis=@n83, Plane X Axis=Normalize(VectorInput=@n84), Plane Center=@n3.RGB, UV Scale=Divide(A=@n86, B=@n87)) {materialFunction=/Game/Athena/Environments/Nexus/Materials/VirtualPlaneCoordinatesAbs}.UVs
   B=@n82
  )
  TextureCoordinate()
  TextureCoordinate()
  TextureCoordinate()
  TextureCoordinate()
  TextureCoordinate()
  VectorParameter() {parameterName=HC_CoreFlameDistortionNoise_Tiling, defaultValue=<rgba 1 1 1 1>}
 ]}
