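# Character material for the Hurricane player body that layers a storm effect onto the shared character shading, for skeletal meshes with morph targets and mesh deformers.
# The base PreFX function feeds MF_Hurricane_Storm, and the base PostFX function finishes the material attributes.
# Static switch HC_UseCoreSprite optionally routes the storm output through the Hurricane head sprite function before PostFX.
# The shading model comes from the graph, using Default Lit and Subsurface Profile.
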Material /BRCosmetics/Characters/Player/Male/Medium/Bodies/M_MED_Hurricane/Materials/M_MED_Hurricane_Storm
graph {
 properties {
  bUseMaterialAttributes=true
  bUsedWithMeshDeformer=true
  bUsedWithMorphTargets=true
  bUsedWithSkeletalMesh=true
  pixelDepthOffsetMode=PDOM_Legacy
  shadingModel=MSM_FromMaterialExpression
  usedShadingModels=Default Lit | Subsurface Profile
 }
 nodes [
  n1 = MaterialFunctionCall() {materialFunction=&BaseFN_CharacterFunction_PreFX}
  n2 = MaterialFunctionCall!(Maerial=@n1.Result) {materialFunction=&/MF_Hurricane_Storm}
  MaterialFunctionCall(
   AO=@n1.AO
   Distance Blend=@n1.AO
   Fresnel=@n1.AO
   Material=StaticSwitchParameter(
    True=MaterialFunctionCall!(CloudsMask=@n2.CloudMask, DirtMask=@n2.CloudMask, Material=@n2.CloudMask) {materialFunction=/BRCosmetics/Characters/Player/Male/Medium/Heads/M_MED_Hurricane_Head_01/Materials/MF_HurricaneHeadSprite}.Result
    False=@n2.Result
   ) {parameterName=HC_UseCoreSprite}
  ) {materialFunction=&BaseFN_CharacterFunction_PostFX}
 ]
}
