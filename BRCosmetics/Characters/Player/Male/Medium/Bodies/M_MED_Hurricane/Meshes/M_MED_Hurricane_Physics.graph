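# Physics asset for the M_MED_Hurricane male player body cosmetic, covering its dynamic accessory bones: gun, torso cable, left and right cables, and three right-side smoke pieces.
# Each accessory pairs a tiny base capsule, constrained to the pelvis capsule with free swing and twist, with a larger capsule on the moving bone.
# Tight per-piece angular limits between each base and its moving bone restrict how far the accessories can sway.
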PhysicsAsset /BRCosmetics/Characters/Player/Male/Medium/Bodies/M_MED_Hurricane/Meshes/M_MED_Hurricane_Physics
bodies {
 dyn_base_leg_cable_a_base_l [{shapeType=Capsule, length=1, radius=1}]
 dyn_base_leg_cable_b_base_l [{shapeType=Capsule, length=1, radius=1}]
 dyn_base_reg_cable_a_base_r [{shapeType=Capsule, length=1, radius=1}]
 dyn_base_reg_cable_b_base_r [{shapeType=Capsule, length=1, radius=1}]
 dyn_gun [{shapeType=Capsule, length=13.9399, radius=3.9559, center=<3.4521 0 0>, rotation=<rot 90 0 0>}]
 dyn_gun_base [{shapeType=Capsule, length=1, radius=1}]
 dyn_leg_cable_a_l [{shapeType=Capsule, length=6, radius=2, center=<5.2019 0 0>, rotation=<rot 90 0 0>}]
 dyn_leg_cable_b_l [{shapeType=Capsule, length=6, radius=2, center=<4.9468 0 0>, rotation=<rot 90 0 0>}]
 dyn_reg_cable_a_r [{shapeType=Capsule, length=6, radius=2, center=<-4.6467 0 0>, rotation=<rot 90 0 0>}]
 dyn_reg_cable_b_r [{shapeType=Capsule, length=6, radius=2, center=<-4.6467 0 0>, rotation=<rot 90 0 0>}]
 dyn_smoke_a_base_r [{shapeType=Capsule, length=1, radius=1}]
 dyn_smoke_a_r [{shapeType=Capsule, length=8, radius=2.25, center=<0 2 0>, rotation=<rot 90 0 90>}]
 dyn_smoke_b_base_r [{shapeType=Capsule, length=1, radius=1}]
 dyn_smoke_b_r [{shapeType=Capsule, length=7.5, radius=2.25, center=<0 2 0>, rotation=<rot 90 0 90>}]
 dyn_smoke_c_base_r [{shapeType=Capsule, length=1, radius=1}]
 dyn_smoke_c_r [{shapeType=Capsule, length=7.75, radius=2, center=<0 2 0>, rotation=<rot 90 0 90>}]
 dyn_torso_cable [{shapeType=Capsule, length=6.5, radius=3.5, center=<12.1155 -0.0001 0>, rotation=<rot 90 0 0>}]
 dyn_torso_cable_base [{shapeType=Capsule, length=1, radius=1}]
 pelvis [{shapeType=Capsule, length=5.25, radius=12}]
}
constraints [
 {
  bone1Name=dyn_torso_cable_base
  bone2Name=pelvis
  swing1LimitDegrees=45
  swing1Motion=Free
  swing2LimitDegrees=45
  swing2Motion=Free
  twistLimitDegrees=45
  twistMotion=Free
 }
 {bone1Name=dyn_torso_cable, bone2Name=dyn_torso_cable_base, swing1LimitDegrees=3, swing2LimitDegrees=30, twistLimitDegrees=3}
 {
  bone1Name=dyn_smoke_a_base_r
  bone2Name=pelvis
  swing1LimitDegrees=45
  swing1Motion=Free
  swing2LimitDegrees=45
  swing2Motion=Free
  twistLimitDegrees=45
  twistMotion=Free
 }
 {
  bone1Name=dyn_smoke_b_base_r
  bone2Name=pelvis
  swing1LimitDegrees=45
  swing1Motion=Free
  swing2LimitDegrees=45
  swing2Motion=Free
  twistLimitDegrees=45
  twistMotion=Free
 }
 {
  bone1Name=dyn_smoke_c_base_r
  bone2Name=pelvis
  swing1LimitDegrees=45
  swing1Motion=Free
  swing2LimitDegrees=45
  swing2Motion=Free
  twistLimitDegrees=45
  twistMotion=Free
 }
 {bone1Name=dyn_smoke_a_r, bone2Name=dyn_smoke_a_base_r, swing1LimitDegrees=5, swing2LimitDegrees=10, twistLimitDegrees=10}
 {bone1Name=dyn_smoke_b_r, bone2Name=dyn_smoke_b_base_r, swing1LimitDegrees=5, swing2LimitDegrees=15, twistLimitDegrees=15}
 {bone1Name=dyn_smoke_c_r, bone2Name=dyn_smoke_c_base_r, swing1LimitDegrees=5, swing2LimitDegrees=5, twistLimitDegrees=5}
 {
  bone1Name=dyn_gun_base
  bone2Name=pelvis
  swing1LimitDegrees=45
  swing1Motion=Free
  swing2LimitDegrees=45
  swing2Motion=Free
  twistLimitDegrees=45
  twistMotion=Free
 }
 {bone1Name=dyn_gun, bone2Name=dyn_gun_base, swing1LimitDegrees=3, swing2LimitDegrees=3, twistLimitDegrees=3}
 {
  bone1Name=dyn_base_leg_cable_a_base_l
  bone2Name=pelvis
  swing1LimitDegrees=45
  swing1Motion=Free
  swing2LimitDegrees=45
  swing2Motion=Free
  twistLimitDegrees=45
  twistMotion=Free
 }
 {
  bone1Name=dyn_base_leg_cable_b_base_l
  bone2Name=pelvis
  swing1LimitDegrees=45
  swing1Motion=Free
  swing2LimitDegrees=45
  swing2Motion=Free
  twistLimitDegrees=45
  twistMotion=Free
 }
 {bone1Name=dyn_leg_cable_a_l, bone2Name=dyn_base_leg_cable_a_base_l, swing1LimitDegrees=10, swing2LimitDegrees=10, twistLimitDegrees=10}
 {bone1Name=dyn_leg_cable_b_l, bone2Name=dyn_base_leg_cable_b_base_l, swing1LimitDegrees=10, swing2LimitDegrees=10, twistLimitDegrees=10}
 {
  bone1Name=dyn_base_reg_cable_a_base_r
  bone2Name=pelvis
  swing1LimitDegrees=45
  swing1Motion=Free
  swing2LimitDegrees=45
  swing2Motion=Free
  twistLimitDegrees=45
  twistMotion=Free
 }
 {
  bone1Name=dyn_base_reg_cable_b_base_r
  bone2Name=pelvis
  swing1LimitDegrees=45
  swing1Motion=Free
  swing2LimitDegrees=45
  swing2Motion=Free
  twistLimitDegrees=45
  twistMotion=Free
 }
 {bone1Name=dyn_reg_cable_a_r, bone2Name=dyn_base_reg_cable_a_base_r, swing1LimitDegrees=10, swing2LimitDegrees=10, twistLimitDegrees=10}
 {bone1Name=dyn_reg_cable_b_r, bone2Name=dyn_base_reg_cable_b_base_r, swing1LimitDegrees=10, swing2LimitDegrees=10, twistLimitDegrees=10}
]
