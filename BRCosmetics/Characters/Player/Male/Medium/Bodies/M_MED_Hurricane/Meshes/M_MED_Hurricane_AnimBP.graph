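# Post-process-style cosmetic AnimBP for the M_MED_Hurricane body, correcting shoulder pad, electrodes, cable and gun attachments atop the parent mesh's copied pose.
# An RBF Pose Driver, weighted constraints and additive bone offsets feed RigidBody simulation.
# The event graph derives offset alphas from limb rotations and fall speed; shoulderUp_R reads clavicle_l.
# Cached shoulder-pad branches switch to a no-twist fix above LOD 1; thigh-volume nodes are unconnected.
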AnimBlueprint /BRCosmetics/Characters/Player/Male/Medium/Bodies/M_MED_Hurricane/Meshes/M_MED_Hurricane_AnimBP
target_skeleton=&/M_MED_Hurricane_Skeleton
graphs [
 AnimGraph {nodes [
   "Constraint - thigh_volume_l (R)"(Alpha=0.1, ConstraintWeights_0=1) {
    alphaBoolBlend {blendOption=Linear}
    boneToModify {boneName=thigh_volume_l}
    constraintSetup [{transformType=Rotation, targetBone {boneName=thigh_l}}]
    constraintWeights [1]
   }
   n1 = "Constraint - thigh_volume_r (R)"(Alpha=0.1, ConstraintWeights_0=1) {
    alphaBoolBlend {blendOption=Linear}
    boneToModify {boneName=thigh_volume_r}
    constraintSetup [{transformType=Rotation, targetBone {boneName=thigh_r}}]
    constraintWeights [1]
   }
   "Transform (Modify) Bone - Bone: thigh_volume_r"(ComponentPose=@n1, Alpha=Get legUp_R() -> (legUp_R:double), Translation="3,3,0", Rotation="0,0,0", Scale="1.2,1.2,1.2") {
    rotationMode=BMM_Replace
    rotationSpace=BCS_BoneSpace
    scaleMode=BMM_Additive
    scaleSpace=BCS_BoneSpace
    translationMode=BMM_Additive
    translationSpace=BCS_BoneSpace
    alphaBoolBlend {blendOption=Linear}
    alphaScaleBiasClamp {bMapRange=true, inRange {max=0, min=-100}, outRange {max=0, min=1}}
    boneToModify {boneName=thigh_volume_r}
    scale=<1.2 1.2 1.2>
    translation=<3 3>
   }
   "Transform (Modify) Bone - Bone: thigh_volume_l"(Alpha=Get legUp_L() -> (legUp_L:double), Translation="-3,-3,0", Rotation="0,0,0", Scale="1.2,1.2,1.2") {
    rotationMode=BMM_Additive
    rotationSpace=BCS_BoneSpace
    scaleMode=BMM_Additive
    scaleSpace=BCS_BoneSpace
    translationMode=BMM_Additive
    translationSpace=BCS_BoneSpace
    alphaBoolBlend {blendOption=Linear}
    alphaScaleBiasClamp {bMapRange=true, inRange {max=0, min=-100}, outRange {max=0, min=1}}
    boneToModify {boneName=thigh_volume_l}
    scale=<1.2 1.2 1.2>
    translation=<-3 -3>
   }
   Output Pose(Result=Blend Poses by bool(
     BlendPose_0=Use cached pose 'NoTwistShoulderPadFixPose'()
     BlendPose_1=Use cached pose 'TwistShoulderPadFixPose'()
     bActiveValue=Get noTwistArm() -> (noTwistArm:bool)
    ) {blendType=Linear}) {layerGroup=None}
   n2 = Pose Driver - Source: upperarm_l!(SourcePose:Pose Link Structure=Copy Pose From Mesh() {bUseAttachedParent=true}) -> (Pose:Pose Link Structure) {
    is=title AnimGraphNode_PoseDriver
    poseAsset=&/M_MED_Hurricane_ShoulderPad_RBF_PoseAsset
    onlyDriveBones [{boneName=dyn_shoulder_pad_r_base}]
    poseTargets [
     {
      bApplyCustomCurve=false
      bIsHidden=false
      distanceMethod=DefaultMethod
      drivenName=Pose_0
      functionType=DefaultFunction
      targetScale=0.9926
      boneTransforms [{targetRotation=<rot -37.7574 0.1509 2.6875>, targetTranslation=<15.5111 0 0>}]
      customCurve {defaultValue=3.4028234663852886e+38, postInfinityExtrap=RCCE_Constant, preInfinityExtrap=RCCE_Constant, keys []}
      targetRotation=<rot 0 0 0>
     }
     {
      bApplyCustomCurve=false
      bIsHidden=false
      distanceMethod=DefaultMethod
      drivenName=Pose_1
      functionType=DefaultFunction
      targetScale=0.8493
      boneTransforms [{targetRotation=<rot 7.6449 2.0825 2.1436>, targetTranslation=<15.5111 0 0>}]
      customCurve {defaultValue=3.4028234663852886e+38, postInfinityExtrap=RCCE_Constant, preInfinityExtrap=RCCE_Constant, keys []}
      targetRotation=<rot 0 0 0>
     }
     {
      bApplyCustomCurve=false
      bIsHidden=false
      distanceMethod=DefaultMethod
      drivenName=Pose_2
      functionType=DefaultFunction
      targetScale=0.9739
      boneTransforms [{targetRotation=<rot 57.226 5.1005 3.9268>, targetTranslation=<15.5111 0 0>}]
      customCurve {defaultValue=3.4028234663852886e+38, postInfinityExtrap=RCCE_Constant, preInfinityExtrap=RCCE_Constant, keys []}
      targetRotation=<rot 0 0 0>
     }
     {
      bApplyCustomCurve=false
      bIsHidden=false
      distanceMethod=DefaultMethod
      drivenName=Pose_3
      functionType=DefaultFunction
      targetScale=0.5974
      boneTransforms [{targetRotation=<rot 28.0414 52.3966 7.0808>, targetTranslation=<15.5111 0 0>}]
      customCurve {defaultValue=3.4028234663852886e+38, postInfinityExtrap=RCCE_Constant, preInfinityExtrap=RCCE_Constant, keys []}
      targetRotation=<rot 0 0 0>
     }
     {
      bApplyCustomCurve=false
      bIsHidden=false
      distanceMethod=DefaultMethod
      drivenName=Pose_4
      functionType=DefaultFunction
      targetScale=0.709
      boneTransforms [{targetRotation=<rot -2.3017 93.1218 7.599>, targetTranslation=<15.5111 0 0>}]
      customCurve {defaultValue=3.4028234663852886e+38, postInfinityExtrap=RCCE_Constant, preInfinityExtrap=RCCE_Constant, keys []}
      targetRotation=<rot 0 0 0>
     }
     {
      bApplyCustomCurve=false
      bIsHidden=false
      distanceMethod=DefaultMethod
      drivenName=Pose_5
      functionType=DefaultFunction
      targetScale=0.58
      boneTransforms [{targetRotation=<rot 3.654 49.1104 7.0518>, targetTranslation=<15.5111 0 0>}]
      customCurve {defaultValue=3.4028234663852886e+38, postInfinityExtrap=RCCE_Constant, preInfinityExtrap=RCCE_Constant, keys []}
      targetRotation=<rot 0 0 0>
     }
     {
      bApplyCustomCurve=false
      bIsHidden=false
      distanceMethod=DefaultMethod
      drivenName=Pose_6
      functionType=DefaultFunction
      targetScale=0.7046
      boneTransforms [{targetRotation=<rot 5.719 -33.0659 -2.0447>, targetTranslation=<15.5111 0 0>}]
      customCurve {defaultValue=3.4028234663852886e+38, postInfinityExtrap=RCCE_Constant, preInfinityExtrap=RCCE_Constant, keys []}
      targetRotation=<rot 0 0 0>
     }
     {
      bApplyCustomCurve=false
      bIsHidden=false
      distanceMethod=DefaultMethod
      drivenName=Pose_7
      functionType=DefaultFunction
      targetScale=0.7046
      boneTransforms [{targetRotation=<rot 5.272 -62.1971 -5.9428>, targetTranslation=<15.5111 0 0>}]
      customCurve {defaultValue=3.4028234663852886e+38, postInfinityExtrap=RCCE_Constant, preInfinityExtrap=RCCE_Constant, keys []}
      targetRotation=<rot 0 0 0>
     }
     {
      bApplyCustomCurve=false
      bIsHidden=false
      distanceMethod=DefaultMethod
      drivenName=Pose_8
      functionType=DefaultFunction
      targetScale=0.6883
      boneTransforms [{targetRotation=<rot -22.8671 -46.383 9.8347>, targetTranslation=<15.5111 0 0>}]
      customCurve {defaultValue=3.4028234663852886e+38, postInfinityExtrap=RCCE_Constant, preInfinityExtrap=RCCE_Constant, keys []}
      targetRotation=<rot 0 0 0>
     }
     {
      bApplyCustomCurve=false
      bIsHidden=false
      distanceMethod=DefaultMethod
      drivenName=Pose_9
      functionType=DefaultFunction
      targetScale=0.7716
      boneTransforms [{targetRotation=<rot -78.4439 -8.6556 10.6642>, targetTranslation=<15.5111 0 0>}]
      customCurve {defaultValue=3.4028234663852886e+38, postInfinityExtrap=RCCE_Constant, preInfinityExtrap=RCCE_Constant, keys []}
      targetRotation=<rot 0 0 0>
     }
     {
      bApplyCustomCurve=false
      bIsHidden=false
      distanceMethod=DefaultMethod
      drivenName=Pose_10
      functionType=DefaultFunction
      targetScale=0.726
      boneTransforms [{targetRotation=<rot -41.935 55.3657 -40.6602>, targetTranslation=<15.5111 0 0>}]
      customCurve {defaultValue=3.4028234663852886e+38, postInfinityExtrap=RCCE_Constant, preInfinityExtrap=RCCE_Constant, keys []}
      targetRotation=<rot 0 0 0>
     }
     {
      bApplyCustomCurve=false
      bIsHidden=false
      distanceMethod=DefaultMethod
      drivenName=Pose_11
      functionType=DefaultFunction
      targetScale=0.6883
      boneTransforms [{targetRotation=<rot -43.4624 -69.9666 69.4913>, targetTranslation=<15.5111 0 0>}]
      customCurve {defaultValue=3.4028234663852886e+38, postInfinityExtrap=RCCE_Constant, preInfinityExtrap=RCCE_Constant, keys []}
      targetRotation=<rot 0 0 0>
     }
     {
      bApplyCustomCurve=false
      bIsHidden=false
      distanceMethod=DefaultMethod
      drivenName=Pose_12
      functionType=DefaultFunction
      targetScale=0.6159
      boneTransforms [{targetRotation=<rot 48.0922 72.4313 -6.6698>, targetTranslation=<15.5111 0 0>}]
      customCurve {defaultValue=3.4028234663852886e+38, postInfinityExtrap=RCCE_Constant, preInfinityExtrap=RCCE_Constant, keys []}
      targetRotation=<rot 0 0 0>
     }
     {
      bApplyCustomCurve=false
      bIsHidden=false
      distanceMethod=DefaultMethod
      drivenName=Pose_13
      functionType=DefaultFunction
      targetScale=0.7716
      boneTransforms [{targetRotation=<rot -69.4979 -172.5092 173.9242>, targetTranslation=<15.5111 0 0>}]
      customCurve {defaultValue=3.4028234663852886e+38, postInfinityExtrap=RCCE_Constant, preInfinityExtrap=RCCE_Constant, keys []}
      targetRotation=<rot 0 0 0>
     }
     {
      bApplyCustomCurve=false
      bIsHidden=false
      distanceMethod=DefaultMethod
      drivenName=Pose_14
      functionType=DefaultFunction
      targetScale=0.58
      boneTransforms [{targetRotation=<rot -13.4769 65.8568 -31.4924>, targetTranslation=<15.5111 0 0>}]
      customCurve {defaultValue=3.4028234663852886e+38, postInfinityExtrap=RCCE_Constant, preInfinityExtrap=RCCE_Constant, keys []}
      targetRotation=<rot 0 0 0>
     }
     {
      bApplyCustomCurve=false
      bIsHidden=false
      distanceMethod=DefaultMethod
      drivenName=Pose_15
      functionType=DefaultFunction
      targetScale=0.9065
      boneTransforms [{targetRotation=<rot -53.1476 109.6244 -57.8883>, targetTranslation=<15.5111 0 0>}]
      customCurve {defaultValue=3.4028234663852886e+38, postInfinityExtrap=RCCE_Constant, preInfinityExtrap=RCCE_Constant, keys []}
      targetRotation=<rot 0 0 0>
     }
     {
      bApplyCustomCurve=false
      bIsHidden=false
      distanceMethod=DefaultMethod
      drivenName=Pose_16
      functionType=DefaultFunction
      targetScale=0.921
      boneTransforms [{targetRotation=<rot 32.4961 118.7317 28.9677>, targetTranslation=<15.5111 0 0>}]
      customCurve {defaultValue=3.4028234663852886e+38, postInfinityExtrap=RCCE_Constant, preInfinityExtrap=RCCE_Constant, keys []}
      targetRotation=<rot 0 0 0>
     }
     {
      bApplyCustomCurve=false
      bIsHidden=false
      distanceMethod=DefaultMethod
      drivenName=Pose_17
      functionType=DefaultFunction
      targetScale=0.835
      boneTransforms [{targetRotation=<rot 39.4008 -57.7573 -45.6614>, targetTranslation=<15.5111 0 0>}]
      customCurve {defaultValue=3.4028234663852886e+38, postInfinityExtrap=RCCE_Constant, preInfinityExtrap=RCCE_Constant, keys []}
      targetRotation=<rot 0 0 0>
     }
     {
      bApplyCustomCurve=false
      bIsHidden=false
      distanceMethod=DefaultMethod
      drivenName=Pose_18
      functionType=DefaultFunction
      targetScale=1
      boneTransforms [{targetRotation=<rot -6.7303 134.1828 4.2185>, targetTranslation=<15.5111 0 0>}]
      customCurve {defaultValue=3.4028234663852886e+38, postInfinityExtrap=RCCE_Constant, preInfinityExtrap=RCCE_Constant, keys []}
      targetRotation=<rot 0 0 0>
     }
    ]
    rBFParams {radius=35}
    sourceBones [{boneName=upperarm_l}]
   }
   n3 = "Constraint - dyn_shldr_electrode_a_base_l (PPP)"!(
    ComponentPose=Local To Component(LocalPose=@n2)
    Alpha=1
    ConstraintWeights_0=0.0162
    ConstraintWeights_1:float=0.3155
    ConstraintWeights_2:float=0.6683
   ) -> (Pose) {
    alphaBoolBlend {blendOption=Linear}
    boneToModify {boneName=dyn_shldr_electrode_a_base_l}
    constraintSetup [{targetBone {boneName=spine_04}}, {targetBone {boneName=spine_05}}, {targetBone {boneName=clavicle_l}}]
    constraintWeights [1,1,1]
   }
   n4 = "Constraint - dyn_shldr_electrode_b_base_l (PPPP)"!(
    ComponentPose=@n3
    Alpha=1
    ConstraintWeights_0=0.0117
    ConstraintWeights_1:float=0.0188
    ConstraintWeights_2:float=0.3509
    ConstraintWeights_3:float=0.6186
   ) -> (Pose) {
    alphaBoolBlend {blendOption=Linear}
    boneToModify {boneName=dyn_shldr_electrode_b_base_l}
    constraintSetup [{targetBone {boneName=spine_03}}, {targetBone {boneName=spine_04}}, {targetBone {boneName=spine_05}}, {targetBone {boneName=clavicle_l}}]
    constraintWeights [1,1,1,1]
   }
   n5 = "Constraint - dyn_shldr_electrode_a_base_r (PPP)"!(ComponentPose=@n4, Alpha=1, ConstraintWeights_0=0.6683, ConstraintWeights_1:float=0.0162, ConstraintWeights_2:float=0.3155) -> (Pose) {
    alphaBoolBlend {blendOption=Linear}
    boneToModify {boneName=dyn_shldr_electrode_a_base_r}
    constraintSetup [{targetBone {boneName=clavicle_r}}, {targetBone {boneName=spine_04}}, {targetBone {boneName=spine_05}}]
    constraintWeights [1,1,1]
   }
   n6 = "Constraint - dyn_shldr_electrode_b_base_r (PPPP)"!(
    ComponentPose=@n5
    Alpha=1
    ConstraintWeights_0=0.6186
    ConstraintWeights_1:float=0.0117
    ConstraintWeights_2:float=0.0188
    ConstraintWeights_3:float=0.3509
   ) -> (Pose) {
    alphaBoolBlend {blendOption=Linear}
    boneToModify {boneName=dyn_shldr_electrode_b_base_r}
    constraintSetup [{targetBone {boneName=clavicle_r}}, {targetBone {boneName=spine_03}}, {targetBone {boneName=spine_04}}, {targetBone {boneName=spine_05}}]
    constraintWeights [1,1,1,1]
   }
   n7 = "Constraint - dyn_shldr_electrode_c_base_r (PPP)"!(ComponentPose=@n6, Alpha=1, ConstraintWeights_0=0.4437, ConstraintWeights_1:float=0.2143, ConstraintWeights_2:float=0.3419) -> (Pose) {
    alphaBoolBlend {blendOption=Linear}
    boneToModify {boneName=dyn_shldr_electrode_c_base_r}
    constraintSetup [{targetBone {boneName=clavicle_r}}, {targetBone {boneName=spine_04}}, {targetBone {boneName=spine_05}}]
    constraintWeights [1,1,1]
   }
   n8 = "Constraint - dyn_base_leg_electrode_fr_a_base_l (PP)"!(ComponentPose=@n7, Alpha=1, ConstraintWeights_0=0.2695, ConstraintWeights_1:float=0.7305) -> (Pose) {
    alphaBoolBlend {blendOption=Linear}
    boneToModify {boneName=dyn_base_leg_electrode_fr_a_base_l}
    constraintSetup [{targetBone {boneName=pelvis}}, {targetBone {boneName=thigh_twist_01_l}}]
    constraintWeights [1,1]
   }
   n9 = "Constraint - dyn_base_leg_electrode_fr_b_base_l (PP)"!(ComponentPose=@n8, Alpha=1, ConstraintWeights_0=0.8329, ConstraintWeights_1:float=0.1671) -> (Pose) {
    alphaBoolBlend {blendOption=Linear}
    boneToModify {boneName=dyn_base_leg_electrode_fr_b_base_l}
    constraintSetup [{targetBone {boneName=thigh_twist_01_l}}, {targetBone {boneName=thigh_l}}]
    constraintWeights [1,1]
   }
   n10 = "Constraint - dyn_base_leg_electrode_bk_a_base_l (PPP)"!(ComponentPose=@n9, Alpha=1, ConstraintWeights_0=0.233, ConstraintWeights_1:float=0.7285, ConstraintWeights_2:float=0.0385) -> (Pose) {
    alphaBoolBlend {blendOption=Linear}
    boneToModify {boneName=dyn_base_leg_electrode_bk_a_base_l}
    constraintSetup [{targetBone {boneName=pelvis}}, {targetBone {boneName=thigh_twist_01_l}}, {targetBone {boneName=thigh_l}}]
    constraintWeights [1,1,1]
   }
   n11 = "Constraint - dyn_base_leg_electrode_bk_b_base_l (PP)"!(ComponentPose=@n10, Alpha=1, ConstraintWeights_0=0.6796, ConstraintWeights_1:float=0.3204) -> (Pose) {
    alphaBoolBlend {blendOption=Linear}
    boneToModify {boneName=dyn_base_leg_electrode_bk_b_base_l}
    constraintSetup [{targetBone {boneName=thigh_twist_01_l}}, {targetBone {boneName=thigh_l}}]
    constraintWeights [1,1]
   }
   n12 = "Constraint - dyn_base_reg_electrode_fr_a_base_r (PP)"!(ComponentPose=@n11, Alpha=1, ConstraintWeights_0=0.2695, ConstraintWeights_1:float=0.7305) -> (Pose) {
    alphaBoolBlend {blendOption=Linear}
    boneToModify {boneName=dyn_base_reg_electrode_fr_a_base_r}
    constraintSetup [{targetBone {boneName=pelvis}}, {targetBone {boneName=thigh_twist_01_r}}]
    constraintWeights [1,1]
   }
   n13 = "Constraint - dyn_base_reg_electrode_fr_b_base_r (PP)"!(ComponentPose=@n12, Alpha=1, ConstraintWeights_0=0.1671, ConstraintWeights_1:float=0.8329) -> (Pose) {
    alphaBoolBlend {blendOption=Linear}
    boneToModify {boneName=dyn_base_reg_electrode_fr_b_base_r}
    constraintSetup [{targetBone {boneName=thigh_r}}, {targetBone {boneName=thigh_twist_01_r}}]
    constraintWeights [1,1]
   }
   n14 = "Constraint - dyn_base_reg_electrode_bk_a_base_r (PPP)"!(ComponentPose=@n13, Alpha=1, ConstraintWeights_0=0.0385, ConstraintWeights_1:float=0.233, ConstraintWeights_2:float=0.7285) -> (Pose) {
    alphaBoolBlend {blendOption=Linear}
    boneToModify {boneName=dyn_base_reg_electrode_bk_a_base_r}
    constraintSetup [{targetBone {boneName=thigh_r}}, {targetBone {boneName=pelvis}}, {targetBone {boneName=thigh_twist_01_r}}]
    constraintWeights [1,1,1]
   }
   n15 = "Constraint - dyn_base_reg_electrode_bk_b_base_r (PP)"!(ComponentPose=@n14, Alpha=1, ConstraintWeights_0=0.3204, ConstraintWeights_1:float=0.6796) -> (Pose) {
    alphaBoolBlend {blendOption=Linear}
    boneToModify {boneName=dyn_base_reg_electrode_bk_b_base_r}
    constraintSetup [{targetBone {boneName=thigh_r}}, {targetBone {boneName=thigh_twist_01_r}}]
    constraintWeights [1,1]
   }
   n16 = "Constraint - dyn_gun_base (PPP)"!(ComponentPose=@n15, Alpha=1, ConstraintWeights_0=0.0268, ConstraintWeights_1:float=0.1258, ConstraintWeights_2:float=0.8474) -> (Pose) {
    alphaBoolBlend {blendOption=Linear}
    boneToModify {boneName=dyn_gun_base}
    constraintSetup [{targetBone {boneName=thigh_r}}, {targetBone {boneName=pelvis}}, {targetBone {boneName=thigh_twist_01_r}}]
    constraintWeights [1,1,1]
   }
   n17 = "Constraint - dyn_torso_cable_base (PPPPPP)"!(
    ComponentPose=@n16
    Alpha=1
    ConstraintWeights_0=0.051
    ConstraintWeights_1:float=0.161
    ConstraintWeights_2:float=0.3736
    ConstraintWeights_3:float=0.2972
    ConstraintWeights_4:float=0.0454
    ConstraintWeights_5:float=0.072
   ) -> (Pose) {
    alphaBoolBlend {blendOption=Linear}
    boneToModify {boneName=dyn_torso_cable_base}
    constraintSetup [
     {targetBone {boneName=clavicle_l}}
     {targetBone {boneName=spine_01}}
     {targetBone {boneName=spine_02}}
     {targetBone {boneName=spine_03}}
     {targetBone {boneName=spine_04}}
     {targetBone {boneName=pelvis}}
    ]
    constraintWeights [1,1,1,1,1,1]
   }
   n18 = "Constraint - dyn_shldr_bk_socket_base_l (PPPPPPP)"!(
    ComponentPose=@n17
    Alpha=1
    ConstraintWeights_0=0.3991
    ConstraintWeights_1:float=0.0287
    ConstraintWeights_2:float=0.027
    ConstraintWeights_3:float=0.1995
    ConstraintWeights_4:float=0.1378
    ConstraintWeights_5:float=0.0151
    ConstraintWeights_6:float=0.1927
   ) -> (Pose) {
    alphaBoolBlend {blendOption=Linear}
    boneToModify {boneName=dyn_shldr_bk_socket_base_l}
    constraintSetup [
     {targetBone {boneName=clavicle_l}}
     {targetBone {boneName=upperarm_twist_02_l}}
     {targetBone {boneName=spine_02}}
     {targetBone {boneName=spine_03}}
     {targetBone {boneName=spine_04}}
     {targetBone {boneName=spine_05}}
     {targetBone {boneName=upperarm_twist_01_l}}
    ]
    constraintWeights [1,1,1,1,1,1,1]
   }
   n19 = "Constraint - dyn_shldr_ft_socket_base_l (PPPPPPP)"!(
    ComponentPose=@n18
    Alpha=1
    ConstraintWeights_0=0.2776
    ConstraintWeights_1:float=0.1232
    ConstraintWeights_2:float=0.051
    ConstraintWeights_3:float=0.0972
    ConstraintWeights_4:float=0.131
    ConstraintWeights_5:float=0.0519
    ConstraintWeights_6:float=0.268
   ) -> (Pose) {
    alphaBoolBlend {blendOption=Linear}
    boneToModify {boneName=dyn_shldr_ft_socket_base_l}
    constraintSetup [
     {targetBone {boneName=clavicle_l}}
     {targetBone {boneName=upperarm_twist_02_l}}
     {targetBone {boneName=spine_02}}
     {targetBone {boneName=spine_03}}
     {targetBone {boneName=spine_04}}
     {targetBone {boneName=spine_05}}
     {targetBone {boneName=upperarm_twist_01_l}}
    ]
    constraintWeights [1,1,1,1,1,1,1]
   }
   n20 = "Transform (Modify) Bone - Bone: dyn_base_leg_electrode_fr_a_base_l"(ComponentPose=@n19, Alpha=Get legUp_L(), Translation="1.5,0,0", Rotation="10,-10,0", Scale="1,1,1") {
    rotationMode=BMM_Additive
    rotationSpace=BCS_BoneSpace
    translationMode=BMM_Additive
    translationSpace=BCS_BoneSpace
    alphaBoolBlend {blendOption=Linear}
    alphaScaleBiasClamp {bMapRange=true, inRange {max=0, min=-100}, outRange {max=0, min=1}}
    boneToModify {boneName=dyn_base_leg_electrode_fr_a_base_l}
    rotation {pitch=10, yaw=-10}
    translation {x=1.5}
   }
   n21 = "Transform (Modify) Bone - Bone: dyn_base_reg_electrode_fr_a_base_r"(ComponentPose=@n20, Alpha=Get legUp_R(), Translation="-1.5,0,0", Rotation="10,-10,0", Scale="1,1,1") {
    rotationMode=BMM_Additive
    rotationSpace=BCS_BoneSpace
    translationMode=BMM_Additive
    translationSpace=BCS_BoneSpace
    alphaBoolBlend {blendOption=Linear}
    alphaScaleBiasClamp {bMapRange=true, inRange {max=0, min=-100}, outRange {max=0, min=1}}
    boneToModify {boneName=dyn_base_reg_electrode_fr_a_base_r}
    rotation {pitch=10, yaw=-10}
    translation {x=-1.5}
   }
   n22 = "Transform (Modify) Bone - Bone: dyn_shldr_bk_socket_base_l"(ComponentPose=@n21, Alpha=Get armUp_L() -> (armUp_L:double), Translation="1,-1,-2", Rotation="-10,0,5", Scale="1,1,1") {
    rotationMode=BMM_Additive
    rotationSpace=BCS_BoneSpace
    translationMode=BMM_Additive
    translationSpace=BCS_BoneSpace
    alphaBoolBlend {blendOption=Linear}
    alphaScaleBiasClamp {bMapRange=true, inRange {max=50}}
    boneToModify {boneName=dyn_shldr_bk_socket_base_l}
    rotation {pitch=-10, roll=5}
    translation=<1 -1 -2>
   }
   n23 = "Transform (Modify) Bone - Bone: dyn_shldr_bk_socket_base_l"(ComponentPose=@n22, Alpha=Get armFwd_L() -> (armFwd_L:double), Translation="0,-1,-2", Rotation="0,-15,0", Scale="1,1,1") {
    rotationMode=BMM_Additive
    rotationSpace=BCS_BoneSpace
    translationMode=BMM_Additive
    translationSpace=BCS_BoneSpace
    alphaBoolBlend {blendOption=Linear}
    alphaScaleBiasClamp {bMapRange=true, inRange {max=70}}
    boneToModify {boneName=dyn_shldr_bk_socket_base_l}
    rotation {yaw=-15}
    translation {y=-1, z=-2}
   }
   n24 = "Transform (Modify) Bone - Bone: dyn_shldr_electrode_a_base_l"(ComponentPose=@n23, Alpha=Get shoulderUp_L() -> (shoulderUp_L:double), Translation="-1,0,3", Rotation="-25,0,0", Scale="1,1,1") {
    rotationMode=BMM_Additive
    rotationSpace=BCS_BoneSpace
    translationMode=BMM_Additive
    translationSpace=BCS_BoneSpace
    alphaBoolBlend {blendOption=Linear}
    alphaScaleBiasClamp {bMapRange=true, inRange {max=45}}
    boneToModify {boneName=dyn_shldr_electrode_a_base_l}
    rotation {pitch=-25}
    translation {x=-1, z=3}
   }
   n25 = "Transform (Modify) Bone - Bone: dyn_shldr_electrode_a_base_r"(ComponentPose=@n24, Alpha=Get shoulderUp_R() -> (shoulderUp_R:double), Translation="1,0,-1.5", Rotation="-25,0,0", Scale="1,1,1") {
    rotationMode=BMM_Additive
    rotationSpace=BCS_BoneSpace
    translationMode=BMM_Additive
    translationSpace=BCS_BoneSpace
    alphaBoolBlend {blendOption=Linear}
    alphaScaleBiasClamp {bMapRange=true, inRange {max=45}}
    boneToModify {boneName=dyn_shldr_electrode_a_base_r}
    rotation {pitch=-25}
    translation {x=1, z=-1.5}
   }
   n26 = "Transform (Modify) Bone - Bone: dyn_torso_cable_base"(ComponentPose=@n25, Alpha=Get jump_alpha() -> (jump_alpha:double), Translation="-2,0,0", Rotation="20,0,0", Scale="1,1,1") {
    rotationMode=BMM_Additive
    rotationSpace=BCS_BoneSpace
    translationMode=BMM_Additive
    translationSpace=BCS_BoneSpace
    alphaBoolBlend {blendOption=Linear}
    alphaScaleBiasClamp {inRange {max=45}}
    boneToModify {boneName=dyn_torso_cable_base}
    rotation {pitch=20}
    translation {x=-2}
   }
   n27 = RigidBody(ComponentPose=@n26) {
    bClampLinearTranslationLimitToRefPose=true
    bOverrideWorldGravity=true
    overridePhysicsAsset=&/M_MED_Hurricane_Physics
    simulationSpace=BaseBoneSpace
    alphaBoolBlend {blendOption=Linear}
    baseBoneRef {boneName=root}
    componentLinearAccScale=<0.5 0.5 0.5>
    componentLinearVelScale=<0.5 0.5 0.5>
    overrideWorldGravity {z=-980}
   }
   Save cached pose 'sharedPose'(Pose=Component To Local(ComponentPose=@n27)) {cacheName=sharedPose}
   n28 = "Constraint - dyn_shoulder_pad_rot_r_base (P)"(ComponentPose=Local To Component(LocalPose=Use cached pose 'sharedPose'()), Alpha=1, ConstraintWeights_0=1) {
    alphaBoolBlend {blendOption=Linear}
    boneToModify {boneName=dyn_shoulder_pad_rot_r_base}
    constraintSetup [{targetBone {boneName=upperarm_twist_01_l}}]
    constraintWeights [1]
   }
   n29 = "Transform (Modify) Bone - Bone: dyn_shoulder_pad_r_base"(ComponentPose=@n28, Alpha=Get armFwd_L(), Translation="0,0,0.5", Rotation="0,0,0", Scale="1,1,1") {
    rotationMode=BMM_Additive
    rotationSpace=BCS_BoneSpace
    translationMode=BMM_Additive
    translationSpace=BCS_BoneSpace
    alphaBoolBlend {blendOption=Linear}
    alphaScaleBiasClamp {bMapRange=true, inRange {max=60}}
    boneToModify {boneName=dyn_shoulder_pad_r_base}
    translation {z=0.5}
   }
   n30 = "Transform (Modify) Bone - Bone: dyn_shoulder_pad_r_base"(ComponentPose=@n29, Alpha=Get armUp_L(), Translation="0,-1,-0.5", Rotation="0,0,0", Scale="1,1,1") {
    rotationMode=BMM_Additive
    rotationSpace=BCS_BoneSpace
    translationMode=BMM_Additive
    translationSpace=BCS_BoneSpace
    alphaBoolBlend {blendOption=Linear}
    alphaScaleBiasClamp {bMapRange=true, inRange {max=45}}
    boneToModify {boneName=dyn_shoulder_pad_r_base}
    translation {y=-1, z=-0.5}
   }
   Save cached pose 'TwistShoulderPadFixPose'(Pose=Component To Local(ComponentPose=@n30)) {cacheName=TwistShoulderPadFixPose}
   n31 = "Transform (Modify) Bone - Bone: dyn_shoulder_pad_r_base"(
    ComponentPose=Local To Component(LocalPose=Use cached pose 'sharedPose'())
    Alpha=Get armFwd_L()
    Translation="0,4,1"
    Rotation="0,0,40"
    Scale="1,1,1"
   ) {
    rotationMode=BMM_Additive
    rotationSpace=BCS_BoneSpace
    translationMode=BMM_Additive
    translationSpace=BCS_BoneSpace
    alphaBoolBlend {blendOption=Linear}
    alphaScaleBiasClamp {bMapRange=true, inRange {max=60}}
    boneToModify {boneName=dyn_shoulder_pad_r_base}
    rotation {roll=40}
    translation {y=4, z=1}
   }
   n32 = "Transform (Modify) Bone - Bone: dyn_shoulder_pad_r_base"(ComponentPose=@n31, Alpha=Get armUp_L(), Translation="0,0,0.5", Rotation="0,0,0", Scale="1,1,1") {
    rotationMode=BMM_Additive
    rotationSpace=BCS_BoneSpace
    translationMode=BMM_Additive
    translationSpace=BCS_BoneSpace
    alphaBoolBlend {blendOption=Linear}
    alphaScaleBiasClamp {bMapRange=true, inRange {max=45}}
    boneToModify {boneName=dyn_shoulder_pad_r_base}
    translation {z=0.5}
   }
   Save cached pose 'NoTwistShoulderPadFixPose'(Pose=Component To Local(ComponentPose=@n32)) {cacheName=NoTwistShoulderPadFixPose}
  ]}
 EventGraph {nodes [
   Event BlueprintUpdateAnimation()
   > n1 = Sequence!(execute) -> (then_0, then_1, then_2:Exec, then_3:Exec)
   > n2 = Sequence!(execute) -> (then_0, then_1, then_2, then_3, then_4:Exec)
   n3 = GetDeltaTransformFromRefPose(self=GetOwnerSkeletalMeshComponent(), BoneName=thigh_l, BaseName=pelvis)
   > Set legUp_L(legUp_L:double=BreakRotator(InRot=BreakTransform(InTransform=@n3).Rotation).Yaw) -> (Output_Get:double)
   n4 = GetDeltaTransformFromRefPose(self=GetOwnerSkeletalMeshComponent(), BoneName=thigh_r, BaseName=pelvis)
   Set legUp_R(execute=@n2.then_1, legUp_R:double=BreakRotator(InRot=BreakTransform(InTransform=@n4).Rotation).Yaw) -> (Output_Get:double)
   n5 = GetOwnerSkeletalMeshComponent()
   n6 = BreakRotator(InRot=BreakTransform(InTransform=GetDeltaTransformFromRefPose(self=@n5, BoneName=upperarm_l, BaseName=clavicle_l)).Rotation)
   Set armUp_L(execute=@n2.then_2, armUp_L:double=@n6.Pitch) -> (Output_Get:double)
   > Set armFwd_L(armFwd_L:double=@n6.Yaw) -> (Output_Get:double)
   n7 = GetDeltaTransformFromRefPose(self=GetOwnerSkeletalMeshComponent(), BoneName=clavicle_l, BaseName=spine_05)
   Set shoulderUp_L(execute=@n2.then_3, shoulderUp_L:double=BreakRotator(InRot=BreakTransform(InTransform=@n7).Rotation).Pitch) -> (Output_Get:double)
   n8 = GetDeltaTransformFromRefPose(self=GetOwnerSkeletalMeshComponent(), BoneName=clavicle_l, BaseName=spine_05)
   Set shoulderUp_R(execute=@n2.then_4, shoulderUp_R:double=BreakRotator(InRot=BreakTransform(InTransform=@n8).Rotation).Pitch) -> (Output_Get:double)
   n9 = Branch(execute=@n1.then_2, Condition="integer > integer"(A=Get CurrentLOD(), B=1) {is=title K2Node_CallFunction})
   > Set noTwistArm(noTwistArm:bool=true) -> (Output_Get:bool)
   Set noTwistArm(execute=@n9.else, noTwistArm=false)
   Set jump_alpha(execute=@n1.then_1, jump_alpha:double=MapRangeClamped(Value=Get PawnVelocityZ(), InRangeA=-700, InRangeB=0, OutRangeA=1, OutRangeB=0)) -> (Output_Get:double)
   TryGetPawnOwner()
  ]}
]
variables [
 armFwd_L {type=number}
 armUp_L {type=number}
 jump_alpha {type=number}
 legUp_L {type=number}
 legUp_R {type=number}
 noTwistArm {type=boolean}
 shoulderUp_L {type=number}
 shoulderUp_R {type=number}
]
